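# Physics asset for the CrispSeason backpack cosmetic, defining collision and joints for a dangling charm on the pack.
# Three capsules on dyn_body approximate the pack, while small capsules cover the charm base and six chain links, the last one larger.
# The base joint to the body swings and twists freely, and each following link chains to the previous one with moderate swing limits and very tight twist.
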PhysicsAsset /BRCosmetics/Accessories/FORT_Backpacks/Backpack_F_MED_CrispSeason/Meshes/F_MED_CrispSeason_Pack_Physics
bodies {
 dyn_body [
  {shapeType=Capsule, length=21.5967, radius=8, shapeName=dyn_body_capsule, center=<0.8023 -6.6047 -5.2769>, rotation=<rot 1.0444 -90 -1.0497>}
  {shapeType=Capsule, length=30, radius=4, shapeName=dyn_body_capsule, center=<-5.6044 -0.5122 -9.96>}
  {shapeType=Capsule, length=30, radius=4, shapeName=dyn_body_capsule, center=<-6.2473 1.8045 -9.96>}
 ]
 dyn_charm_01 [{shapeType=Capsule, length=0, radius=0.6, shapeName=dyn_charm_01_capsule, center=<0.2431 0 0>}]
 dyn_charm_02 [{shapeType=Capsule, length=0, radius=0.8, shapeName=dyn_charm_02_capsule, center=<0.2431 0 0>}]
 dyn_charm_03 [{shapeType=Capsule, length=0, radius=0.8, shapeName=dyn_charm_03_capsule, center=<0.2431 0 0>}]
 dyn_charm_04 [{shapeType=Capsule, length=0, radius=0.8, shapeName=dyn_charm_04_capsule, center=<0.2431 0 0>}]
 dyn_charm_05 [{shapeType=Capsule, length=0, radius=0.8, shapeName=dyn_charm_05_capsule, center=<0.2431 0 0>}]
 dyn_charm_06 [{shapeType=Capsule, length=1, radius=2, shapeName=dyn_charm_06_capsule, center=<1 0 0>, rotation=<rot 90 0 0>}]
 dyn_charm_basa [{shapeType=Capsule, length=0.2, radius=0.2, shapeName=dyn_charm_basa_capsule}]
}
constraints [
 {
  bone1Name=dyn_charm_basa
  bone2Name=dyn_body
  swing1LimitDegrees=45
  swing1Motion=Free
  swing2LimitDegrees=45
  swing2Motion=Free
  twistLimitDegrees=45
  twistMotion=Free
 }
 {bone1Name=dyn_charm_01, bone2Name=dyn_charm_basa, swing1LimitDegrees=35, swing2LimitDegrees=35, twistLimitDegrees=3}
 {bone1Name=dyn_charm_02, bone2Name=dyn_charm_01, swing1LimitDegrees=35, swing2LimitDegrees=35, twistLimitDegrees=3}
 {bone1Name=dyn_charm_03, bone2Name=dyn_charm_02, swing1LimitDegrees=35, swing2LimitDegrees=35, twistLimitDegrees=3}
 {bone1Name=dyn_charm_04, bone2Name=dyn_charm_03, swing1LimitDegrees=35, swing2LimitDegrees=35, twistLimitDegrees=3}
 {bone1Name=dyn_charm_05, bone2Name=dyn_charm_04, swing1LimitDegrees=35, swing2LimitDegrees=35, twistLimitDegrees=3}
 {bone1Name=dyn_charm_06, bone2Name=dyn_charm_05, swing1LimitDegrees=35, swing2LimitDegrees=35, twistLimitDegrees=3}
]
